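# Niagara system for the Cherry glider's idle effect, layering glows, wisps, sparkles and ribbon trails on skeletal meshes.
# One-shot burst and looping Corona, Wispy and SparkleDust emitters pin particles to bones by spawn group, mostly timed by User.InitDelay.
# Attached ribbons bend with ForwardDot and a front-end preview flag.
# BoneFlow emitters stream sprites along the main skeleton and drive pendulum-constrained ribbons and sprites on a separate hair skeleton.
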NiagaraSystem /BRCosmetics/Gadgets/Assets/VinderTech_GliderChute/Glider_Cherry/FX/NS_Glider_Cherry_Idle
dependencies {
 usedDataInterfaces [&NiagaraDataInterfaceColorCurve, &NiagaraDataInterfaceCurve]
 usedDynamicInputs []
 usedModules [&SystemState]
}
emitters {
 Corona_BURST {
  emitterUpdateScript [
   &EmitterState(
    Inactive Response=::NewEnumerator2
    Life Cycle Mode=::NewEnumerator1
    Loop Behavior=::NewEnumerator1
    Loop Delay=$User.InitDelay
    Loop Duration=0.5
    UseLoopDelay=true
   )
   &SpawnBurst_Instantaneous(Spawn Count=5, Use Spawn Probability=false)
  ]
  particleSpawnScript [
   /Niagara/Modules/Spawn/Initialization/InitializeParticle(
    Color=<rgba 1 1.2 2 1>
    Lifetime=0.3
    Module.Write Color=true
    Module.Write Lifetime=true
    Module.Write Mass=true
    Module.Write Material Random=true
    Module.Write Position=false
    Module.Write Scale=false
    Module.Write SpriteRotation=false
    Module.Write SpriteSize=true
    Sprite Size=&Vector2DFromFloat(Value=&RandomRangeFloat(Minimum=100, Maximum=135))
   )
   /SetVariables(Particles.Color2=&ColorFromCurve(
     ColorCurve=DataInterface() {propertyValues {
       AlphaCurve {keys [<at 0 1>,<at 1 1>]}
       BlueCurve {keys [<at 0.0017 0.856>,<at 0.0873 1>,<at 0.201 1>,<at 0.3013 1>,<at 0.4137 0.9434>,<at 1 0.245>]}
       GreenCurve {keys [<at 0.0017 1>,<at 0.0873 0.0354>,<at 0.201 0>,<at 0.3013 0>,<at 0.4137 1>,<at 1 0>]}
       RedCurve {keys [<at 0.0017 0>,<at 0.0873 0>,<at 0.201 0.8276>,<at 0.3013 0.0787>,<at 0.4137 0>,<at 1 1>]}
      }}
     CurveIndex=$Particles.MaterialRandom
    )) {enabled=false, isSetParameters=true, module=SetVariables_#1}
   /SetVariables(Particles.Color2=<rgba 0 0.5486 1 1>) {isSetParameters=true, module=SetVariables_#2}
   /SetVariables(Particles.SpriteSize2=&Multiply_Vector2(A=$Particles.SpriteSize, B=&Vector2DFromFloat(Value=3))) {isSetParameters=true, module=SetVariables_#3}
   &FN_SetSpawnGroup(Number of Spawn Groups=5)
  ]
  particleUpdateScript [
   &SkeletalMeshLocation(
    Bone / Socket Index=&Add_Integer(A=&Add_Integer(A=$Particles.SpawnGroup, B=4), B=15)
    Bone Sampling Mode=::NewEnumerator4
    Inherit Velocity=false
    MaskSpawnGroup=false
    Skeletal Mesh=$User.SkeletalMesh
   )
   &ParticleState()
   &ScaleColor(
    Scale Alpha=&FloatFromCurve(
     FloatCurve=DataInterface() {propertyValues {Curve {keys [
         {arriveTangent=254.6135, interpMode=RCIM_Cubic, leaveTangent=254.6135, tangentMode=RCTM_User, time=0, value=0}
         {arriveTangent=-1.1136, interpMode=RCIM_Cubic, leaveTangent=-3.5209, tangentMode=RCTM_Break, time=0.1881, value=1}
         <at 1 0>
        ]}}}
     Scale Curve=2
    )
    ScaleA=true
    ScaleRGB=true
   )
   /SetVariables(Particles.Color2=&MakeLinearColorFromVectorAndFloat(
     "Vector (RGB)"=&MakeVectorFromLinearColorRGB(Color=$Particles.Color2)
     "Float (Alpha)"=&FloatFromCurve(
      FloatCurve=DataInterface() {propertyValues {Curve {keys [
          {interpMode=RCIM_Cubic, tangentMode=RCTM_Break, time=0, value=0}
          {interpMode=RCIM_Cubic, tangentMode=RCTM_Break, time=0.5, value=1}
          {interpMode=RCIM_Cubic, tangentMode=RCTM_Break, time=1, value=0}
         ]}}}
      Scale Curve=0.125
     )
    )) {isSetParameters=true, module=SetVariables_#4}
   &DynamicMaterialParameters(
    Index 0 Param 1=&RandomRangeFloat(Minimum=0.8, Maximum=1.2)
    Index 0 Param 2=&RandomRangeFloat(Minimum=0.05, Maximum=0.1)
    Index 0 Param 3=&RandomRangeFloat(Minimum=0.4, Maximum=0.6)
    Index 0 Param 4=1
    Module.Param0WriteEnabled=true
    Module.Param1WriteEnabled=true
    Module.Param2WriteEnabled=true
    Module.Param3WriteEnabled=true
   )
   &CameraOffset(Camera Offset Amount=18)
   &ScaleSpriteSize(Scale Factor=&Vector2DFromFloat(Value=&FloatFromCurve(FloatCurve=DataInterface() {propertyValues {Curve {keys [{interpMode=RCIM_Cubic, leaveTangent=1.9829, tangentMode=RCTM_Break, time=0, value=0}, <at 1 1>]}}})))
  ]
  rendererClasses [&NiagaraSpriteRendererProperties]
  renderers [
   {rendererClass=&NiagaraSpriteRendererProperties}
   {rendererClass=&NiagaraSpriteRendererProperties, rendererIndex=1}
  ]
 }
 Corona_Constant {
  emitterUpdateScript [
   &EmitterState(
    Delay First Loop Only=true
    Inactive Response=::NewEnumerator2
    Life Cycle Mode=::NewEnumerator1
    Loop Delay=$User.InitDelay
    Loop Duration=0.5
    UseLoopDelay=true
   )
   &SpawnBurst_Instantaneous(Spawn Count=5, Use Spawn Probability=false)
  ]
  particleSpawnScript [
   &InitializeParticle(
    Color=<rgba 1 1.2 2 0.25>
    Lifetime=1.1
    Module.Write Color=true
    Module.Write Lifetime=true
    Module.Write Mass=true
    Module.Write Material Random=true
    Module.Write Position=false
    Module.Write Scale=false
    Module.Write SpriteRotation=false
    Module.Write SpriteSize=true
    Sprite Size=&Vector2DFromFloat(Value=&RandomRangeFloat(Minimum=45, Maximum=65))
   )
   /SetVariables(Particles.Color2=&ColorFromCurve(
     ColorCurve=DataInterface() {propertyValues {
       AlphaCurve {keys [<at 0 1>,<at 1 1>]}
       BlueCurve {keys [<at 0.0017 0.856>,<at 0.0873 1>,<at 0.201 1>,<at 0.3013 1>,<at 0.4137 0.9434>,<at 1 0.245>]}
       GreenCurve {keys [<at 0.0017 1>,<at 0.0873 0.0354>,<at 0.201 0>,<at 0.3013 0>,<at 0.4137 1>,<at 1 0>]}
       RedCurve {keys [<at 0.0017 0>,<at 0.0873 0>,<at 0.201 0.8276>,<at 0.3013 0.0787>,<at 0.4137 0>,<at 1 1>]}
      }}
     CurveIndex=$Particles.MaterialRandom
    )) {enabled=false, isSetParameters=true, module=SetVariables_#1}
   /SetVariables(Particles.Color2=<rgba 0 0.5486 1 1>) {isSetParameters=true, module=SetVariables_#2}
   /SetVariables(Particles.SpriteSize2=&Multiply_Vector2(A=$Particles.SpriteSize, B=&Vector2DFromFloat(Value=3))) {isSetParameters=true, module=SetVariables_#3}
   &FN_SetSpawnGroup(Number of Spawn Groups=5)
  ]
  particleUpdateScript [
   &SkeletalMeshLocation(
    Bone / Socket Index=&Add_Integer(A=&Add_Integer(A=$Particles.SpawnGroup, B=4), B=15)
    Bone Sampling Mode=::NewEnumerator4
    Inherit Velocity=false
    MaskSpawnGroup=false
    Skeletal Mesh=$User.SkeletalMesh
   )
   &ParticleState()
   &ScaleColor(
    Scale Alpha=&FloatFromCurve(FloatCurve=DataInterface() {propertyValues {Curve {keys [
         <at 0 0 RCIM_Cubic>
         {arriveTangent=-1.1136, interpMode=RCIM_Cubic, leaveTangent=-1.1136, tangentMode=RCTM_Break, time=0.5, value=1}
         <at 1 0>
        ]}}})
    ScaleA=true
    ScaleRGB=true
   )
   /SetVariables(Particles.Color2=&MakeLinearColorFromVectorAndFloat(
     "Vector (RGB)"=&MakeVectorFromLinearColorRGB(Color=$Particles.Color2)
     "Float (Alpha)"=&FloatFromCurve(
      FloatCurve=DataInterface() {propertyValues {Curve {keys [
          {interpMode=RCIM_Cubic, tangentMode=RCTM_Break, time=0, value=0}
          {interpMode=RCIM_Cubic, tangentMode=RCTM_Break, time=0.5, value=1}
          {interpMode=RCIM_Cubic, tangentMode=RCTM_Break, time=1, value=0}
         ]}}}
      Scale Curve=0.125
     )
    )) {isSetParameters=true, module=SetVariables_#4}
   &DynamicMaterialParameters(
    Index 0 Param 1=&RandomRangeFloat(Minimum=0.8, Maximum=1.2)
    Index 0 Param 2=&RandomRangeFloat(Minimum=0.05, Maximum=0.1)
    Index 0 Param 3=&RandomRangeFloat(Minimum=0.4, Maximum=0.6)
    Index 0 Param 4=1
    Module.Param0WriteEnabled=true
    Module.Param1WriteEnabled=true
    Module.Param2WriteEnabled=true
    Module.Param3WriteEnabled=true
   )
   &CameraOffset(Camera Offset Amount=18)
   &ScaleSpriteSize(Scale Factor=&Vector2DFromFloat(Value=/Game/Effects/Niagara/DynamicInputs/Skydiving/SkydiveHandsFeetFloatModifier(Hands:NiagaraFloat=1, Feet:NiagaraFloat=0.75)))
  ]
  rendererClasses [&NiagaraSpriteRendererProperties]
  renderers [
   {rendererClass=&NiagaraSpriteRendererProperties}
   {rendererClass=&NiagaraSpriteRendererProperties, rendererIndex=1}
  ]
 }
 FN_AttachedRibbon_v3_HANDS_2 {
  emitterUpdateScript [
   &EmitterState(Delay First Loop Only=true, Life Cycle Mode=::NewEnumerator1, Loop Delay=$User.InitDelay, Loop Duration=1, UseLoopDelay=true)
   &FN_SpawnRibbon_v2(Number of Ribbons=2, Spawn Mode=::NewEnumerator0, SpawnRate=60)
   &FN_Integrator() {enabled=false}
  ]
  particleSpawnScript [
   &self:FN_InitializeRibbon_v4(
    Enable Skeletal Position=true
    Lifetime=0.3
    Offset Axis:ENiagaraOrientationAxis=::ZAxis
    Ribbon Width=48
    SkeletalMesh:NiagaraDataInterfaceSkeletalMesh=$User.SkeletalMesh
   ) {module=FN_InitializeRibbon_v4001}
   /SetVariables(Particles.SpawnGroup=&Add_Integer(A=$Particles.SpawnGroup, B=15)) {isSetParameters=true}
   &FN_FlipRibbonV()
   &AddVelocity(Velocity=<0 0 500>) {enabled=false}
  ]
  particleUpdateScript [
   &ScaleVelocity(Velocity Scale=<0 0 0>)
   /SetVariables(Particles.RibbonWidth_Hot=30) {isSetParameters=true}
   &AddVelocity(Velocity=&Multiply_VectorByFloat(
     Vector=&Lerp_Vector(
      A=<0 500 -500>
      B=<0 400 500>
      Alpha=&MakeFloatFromInt(Int=&ModuloInt(A=$Particles.SpawnGroup, B=2))
     )
     Float=&FloatFromCurve(FloatCurve=DataInterface() {propertyValues {Curve {keys [
          {interpMode=RCIM_Cubic, leaveTangent=-3.446, tangentMode=RCTM_Break, time=0, value=1}
          {arriveTangent=3.7861, interpMode=RCIM_Cubic, leaveTangent=3.7861, tangentMode=RCTM_User, time=0.4602, value=-0.2603}
          <at 1 -0.6637>
         ]}}})
    )) {enabled=false, module=AddVelocity001}
   &AddVelocity(Velocity=&Multiply_Vector(
     A=&Lerp_Vector(
      A=&MakeVector(
       X=&MakeCustomFloatFromBool(Bool=$User.bIsFrontEndPreview, True Float=400, False Float=0)
       Y=-1500
       Z=&Lerp_Float(A=500, B=50, Alpha=$User.ForwardDot)
      )
      B=&MakeVector(
       X=&MakeCustomFloatFromBool(Bool=$User.bIsFrontEndPreview, True Float=400, False Float=0)
       Y=-1500
       Z=&Lerp_Float(A=-500, B=-50, Alpha=$User.ForwardDot)
      )
      Alpha=&MakeFloatFromInt(Int=&ModuloInt(A=$Particles.SpawnGroup, B=1))
     )
     B=&MakeVector(
      X=1
      Y=0.4
      Z=&FloatFromCurve(FloatCurve=DataInterface() {propertyValues {Curve {keys [
           {interpMode=RCIM_Cubic, leaveTangent=-0.5115, tangentMode=RCTM_Break, time=0, value=1}
           {arriveTangent=-0.1994, interpMode=RCIM_Cubic, tangentMode=RCTM_Break, time=1, value=-0.6637}
          ]}}})
     )
    )) {module=AddVelocity002}
   &SolveForcesAndVelocity(Clamp Velocity=false, Limit Acceleration=false)
   &self:FN_UpdateRibbon_v4(Enable Skeletal Position=true, SkeletalMesh:NiagaraDataInterfaceSkeletalMesh=$User.SkeletalMesh) {module=FN_UpdateRibbon_v4001}
   /Game/Effects/Niagara/Modules/Materials/FN_RibbonMaterial/FN_RibbonMaterial_v2(Color Pan U Speed=1.25, Opacity Pan U Speed=0.8, Write to Module Namespace=true)
   &FN_RibbonMaterial_v2(
    Color Pan U Speed=1.75
    Color Scale U=2
    Opacity Pan U Offset=0.4
    Opacity Pan U Speed=1.4
    Opacity Scale U=2
    Write to Module Namespace=true
   ) {module=FN_RibbonMaterial_v2001}
   &ScaleColor(
    Scale Alpha=&Multiply_Float(
     A=&FloatFromCurve(FloatCurve=DataInterface() {propertyValues {Curve {keys [
          {arriveTangent=9.0725, interpMode=RCIM_Cubic, leaveTangent=9.0725, tangentMode=RCTM_User, time=0, value=0}
          {arriveTangent=-2.5249, interpMode=RCIM_Cubic, leaveTangent=-2.5249, tangentMode=RCTM_User, time=0.1408, value=0.9847}
          {arriveTangent=-0.0405, interpMode=RCIM_Cubic, leaveTangent=-0.0405, tangentMode=RCTM_User, time=1, value=0}
         ]}}})
     B=&FloatFromCurve(
      FloatCurve=DataInterface() {propertyValues {Curve {keys [{interpMode=RCIM_Cubic, leaveTangent=4.3733, tangentMode=RCTM_Break, time=0.9, value=0}, <at 1.4 1>]}}}
      CurveIndex=$Engine.System.Age
     )
    )
    ScaleA=true
    ScaleRGB=true
   )
  ]
  rendererClasses [&NiagaraRibbonRendererProperties]
  renderers [
   {rendererClass=&NiagaraRibbonRendererProperties}
   {rendererClass=&NiagaraRibbonRendererProperties, rendererIndex=1}
  ]
 }
 FN_BoneFlow_HEAD {
  emitterSpawnScript [
   /SetVariables(Emitter.RandomSeed:NiagaraInt32=/Niagara/DynamicInputs/Random/RandomInt(Randomness Mode:ENiagaraRandomnessMode=::NewEnumerator0, Random Seed:NiagaraInt32=0, Fixed Random Seed:NiagaraBool=false)) {isSetParameters=true, module=SetVariables_#5}
   /SetVariables(Emitter.ParticlesPerGroup=5) {isSetParameters=true, module=SetVariables_#6}
  ]
  emitterUpdateScript [
   /SetVariables(Emitter.NumberOfGroups=1) {isSetParameters=true, module=SetVariables_#7}
   &EmitterState(
    Life Cycle Mode=::NewEnumerator1
    Loop Behavior=::NewEnumerator1
    Loop Delay=$User.InitDelay
    Loop Duration Mode=::NewEnumerator1
    UseLoopDelay=false
   )
   &SpawnBurst_Instantaneous(
    Spawn Count=&Multiply_Int(A=$Emitter.ParticlesPerGroup, B=$Emitter.NumberOfGroups)
    Use Spawn Probability=false
   )
  ]
  particleSpawnScript [
   /SetVariables(Particles.LoopCount:NiagaraInt32=$Emitter.LoopCount) {isSetParameters=true, module=SetVariables_#8}
   &FN_SetSpawnGroup(
    Inclusivity=true
    Normalize Spawn Group=true
    Particles Per Group:NiagaraInt32=$Emitter.ParticlesPerGroup
    Pattern=::NewEnumerator0
    Set Ribbon ID and Ribbon Link Order=true
   )
   &InitializeRibbon(
    Lifetime=0.5
    Module.Write Color=true
    Module.Write Lifetime=true
    Module.Write Mass=true
    Module.Write Material Random=true
    Module.Write Position=false
    Module.Write RibbonTwist=false
    Module.Write RibbonWidth=true
    Ribbon Width=30
   )
   &FN_FlipRibbonV()
  ]
  particleUpdateScript [
   &ParticleState(Kill Particles When Lifetime Has Elapsed=false, Let Infinitely Lived Particles Die When Emitter Deactivates=true)
   /Game/Effects/Niagara/Modules/BoneFlow/FN_BoneFlow_v4/FN_BoneFlow(
    BoneID=0
    Height:NiagaraFloat=&Multiply_Float(A=$Particles.Normalized Spawn Group, B=1.85)
    Interpolate Mesh Sampling=false
    SkeletalMesh:NiagaraDataInterfaceSkeletalMesh=$User.SkeletonHair
   ) {module=FN_BoneFlow001}
   &PendulumSetup(
    Module.Enable Spring Driver=false
    Pendulum Length:NiagaraFloat=&FloatFromCurve(
     FloatCurve=DataInterface() {propertyValues {Curve {keys [
         {interpMode=RCIM_Cubic, tangentMode=RCTM_Break, time=0, value=0}
         {arriveTangent=0.6756, interpMode=RCIM_Cubic, leaveTangent=0.6756, tangentMode=RCTM_User, time=1, value=0.2}
        ]}}}
     CurveIndex=$Particles.Normalized Spawn Group
     Scale Curve=50
    )
    Pendulum Pivot:Vector3f=$Output.FN_BoneFlow001.Position
    Rigid Pendulum=false
   )
   &SolveForcesAndVelocity(Clamp Velocity=false, Limit Acceleration=false)
   &PendulumConstraint()
   &CalculateAccurateVelocity()
   /SetVariables(Particles.RibbonFacing=$Engine.Owner.SystemYAxis) {isSetParameters=true, module=SetVariables_#9}
   &FN_RibbonMaterial(Color Pan U Speed=1.25, Opacity Pan U Speed=0.8, UV Basis:NiagaraFloat=$Particles.Normalized Spawn Group)
   &FN_RibbonMaterial(Color Pan U Speed=1.75, Opacity Pan U Speed=1.4, UV Basis:NiagaraFloat=$Particles.Normalized Spawn Group) {module=FN_RibbonMaterial001}
   &ScaleColor(
    Scale Alpha=&FloatFromCurve(
     FloatCurve=DataInterface() {propertyValues {Curve {keys [{interpMode=RCIM_Cubic, leaveTangent=6.8443, tangentMode=RCTM_Break, time=0.9, value=0}, <at 1.4 1>]}}}
     CurveIndex=$Engine.System.Age
     Scale Curve=1.45
    )
    ScaleA=true
    ScaleRGB=true
   )
  ]
  rendererClasses [&NiagaraRibbonRendererProperties]
  renderers [
   {rendererClass=&NiagaraRibbonRendererProperties}
   {rendererClass=&NiagaraRibbonRendererProperties, rendererIndex=1}
   {rendererClass=&NiagaraRibbonRendererProperties, rendererIndex=2}
   {rendererClass=&NiagaraRibbonRendererProperties, rendererIndex=3}
  ]
 }
 FN_BoneFlow_Sprite_HAIR {
  emitterUpdateScript [
   &EmitterState(Life Cycle Mode=::NewEnumerator1, Loop Duration=1, UseLoopDelay=false)
   &SpawnRate(SpawnRate=5, Use Spawn Probability=false)
  ]
  particleSpawnScript [
   &FN_SetSpawnGroup(Number of Spawn Groups=1)
   /Niagara/Modules/Spawn/Initialization/V2/InitializeParticle(
    Color=<rgba 1 1 1 0.45>
    Lifetime Mode=::NewEnumerator1
    Sprite Rotation Angle=&RandomRangeFloat(Minimum=-1, Maximum=1)
    Sprite Rotation Mode=::NewEnumerator1
    Sprite Size Mode=::NewEnumerator1
    Uniform Sprite Size Max=55
    Uniform Sprite Size Min=35
    UsePositionOffset=false
    Write Material Random=true
   )
   &SubUVAnimation(End Frame=4, SubUV Animation Mode:ENiagara_SubUVLookupMode=::NewEnumerator2)
  ]
  particleUpdateScript [
   &ParticleState()
   &FN_BoneFlow(
    Angle:NiagaraFloat=&RandomRangeFloat(Minimum=0, Maximum=24)
    BoneID:NiagaraInt32=$Particles.SpawnGroup
    Height:NiagaraFloat=$Particles.NormalizedAge
    Interpolate Mesh Sampling=false
    Radius:NiagaraFloat=&FloatFromCurve(
     FloatCurve=DataInterface() {propertyValues {Curve {keys [{interpMode=RCIM_Cubic, leaveTangent=-1.5983, tangentMode=RCTM_Break, time=0, value=1}, <at 1 0>]}}}
     Scale Curve=6
    )
    SideVector1:Vector3f=<0 0 1>
    SideVector2:Vector3f=<0 0 1>
    SideVector3:Vector3f=<0 0 1>
    SkeletalMesh:NiagaraDataInterfaceSkeletalMesh=$User.SkeletonHair
    Stabalize Twisting=true
   ) {module=FN_BoneFlow001}
   &PendulumSetup(
    Module.Enable Spring Driver=false
    Pendulum Length:NiagaraFloat=&FloatFromCurve(
     FloatCurve=DataInterface() {propertyValues {Curve {keys [{interpMode=RCIM_Cubic, tangentMode=RCTM_Break, time=0, value=0}, <at 1 0.2>]}}}
     Scale Curve=10
    )
    Pendulum Pivot:Vector3f=$Output.FN_BoneFlow001.Position
    Rigid Pendulum=false
   )
   &GravityForce(Gravity=<0 0 1500>)
   &SolveForcesAndVelocity(Clamp Velocity=false, Limit Acceleration=false)
   &PendulumConstraint()
   &CalculateAccurateVelocity()
   &ScaleColor(
    Scale Alpha=&FloatFromCurve(FloatCurve=DataInterface() {propertyValues {Curve {keys [
         {arriveTangent=4.6219, interpMode=RCIM_Cubic, leaveTangent=4.6219, tangentMode=RCTM_User, time=0, value=0}
         {arriveTangent=2.4948, interpMode=RCIM_Cubic, leaveTangent=2.4948, time=0.1713, value=1}
         {arriveTangent=-1.2067, interpMode=RCIM_Cubic, leaveTangent=-4.5177, tangentMode=RCTM_Break, time=0.3761, value=0.9382}
         <at 1 0>
        ]}}})
    Scale RGB=&MakeVectorFromLinearColorRGB(Color=&ColorFromCurve(ColorCurve=DataInterface() {propertyValues {
        AlphaCurve {keys [<at 0 1 RCIM_Cubic>]}
        BlueCurve {keys [<at -0.4113 0.4974>,<at 0 0.0104 RCIM_Cubic>,<at 0.5561 1 RCIM_Cubic>]}
        GreenCurve {keys [<at -0.4113 1>,<at 0 0.8867 RCIM_Cubic>,<at 0.5561 0.4156 RCIM_Cubic>]}
        RedCurve {keys [<at -0.4113 0>,<at 0 1 RCIM_Cubic>,<at 0.5561 0 RCIM_Cubic>]}
       }}))
    ScaleA=true
    ScaleRGB=true
   )
   &SpriteRotationRate(Rotation Rate=&FloatFromCurve(
     FloatCurve=DataInterface() {propertyValues {Curve {keys [{interpMode=RCIM_Cubic, tangentMode=RCTM_Break, time=0, value=1}, {interpMode=RCIM_Cubic, tangentMode=RCTM_Break, time=1, value=0.489}]}}}
     Scale Curve=&Lerp_Float(A=-80, B=80, Alpha=$Particles.Initial.MaterialRandom)
    ))
   &DynamicMaterialParameters(
    Index 0 Param 1=&FloatFromCurve(FloatCurve=DataInterface() {propertyValues {Curve {keys [{interpMode=RCIM_Cubic, leaveTangent=0.6389, tangentMode=RCTM_Break, time=0, value=0.5}, <at 1 1.75>]}}})
    Index 0 Param 2=&FloatFromCurve(FloatCurve=DataInterface() {propertyValues {Curve {keys [{interpMode=RCIM_Cubic, leaveTangent=0.6389, tangentMode=RCTM_Break, time=0, value=0.5}, <at 1 1.75>]}}})
    Module.Param0WriteEnabled=true
    Module.Param1WriteEnabled=true
    Module.Param2WriteEnabled=true
    Module.Param3WriteEnabled=true
   )
   &ScaleSpriteSize(Scale Factor=&Vector2DFromFloat(Value=&FloatFromCurve(FloatCurve=DataInterface() {propertyValues {Curve {keys [{interpMode=RCIM_Cubic, leaveTangent=0.0179, tangentMode=RCTM_Break, time=0, value=0.55}, <at 1 1>]}}})))
  ]
  rendererClasses [&NiagaraSpriteRendererProperties]
  renderers [{rendererClass=&NiagaraSpriteRendererProperties}]
 }
 FN_BoneFlow_v5_BETA {
  emitterSpawnScript [
   /SetVariables(Emitter.RandomSeed:NiagaraInt32=&RandomInt(Randomness Mode:ENiagaraRandomnessMode=::NewEnumerator0, Random Seed:NiagaraInt32=0, Fixed Random Seed:NiagaraBool=false)) {isSetParameters=true, module=SetVariables_#5}
   /SetVariables(Emitter.ParticlesPerGroup=HlslExpression() {hlslExpression="Engine.QualityLevel > 2 ? 20 : 10"}) {isSetParameters=true, module=SetVariables_#6}
  ]
  emitterUpdateScript [
   /SetVariables(Emitter.NumberOfGroups:NiagaraInt32=&RandomRangeInt(
     Minimum=1
     Maximum=HlslExpression() {hlslExpression="Engine.QualityLevel > 2 ? 3 : 1"}
     Random Seed=$Emitter.LoopCount
     Fixed Random Seed=true
    )) {enabled=false, isSetParameters=true, module=SetVariables_#7}
   &EmitterState(Delay First Loop Only=true, Life Cycle Mode=::NewEnumerator1, Loop Delay=$User.InitDelay, Loop Duration=1, UseLoopDelay=true)
   &SpawnBurst_Instantaneous(
    Spawn Count=&Multiply_Int(A=$Emitter.ParticlesPerGroup, B=$Emitter.NumberOfGroups)
    Use Spawn Probability=false
   ) {enabled=false}
   &SpawnRate(SpawnRate=24, Use Spawn Probability=false)
  ]
  particleSpawnScript [
   /Niagara/Modules/Spawn/Initialization/InitializeParticle(
    Color=<rgba 1 1 1 0.25>
    Module.Write Color=true
    Module.Write Lifetime=true
    Module.Write Mass=true
    Module.Write Material Random=true
    Module.Write Position=false
    Module.Write Scale=false
    Module.Write SpriteRotation=true
    Module.Write SpriteSize=true
    Sprite Rotation=&RandomRangeFloat(Minimum=-360, Maximum=360)
    Sprite Size=&Vector2DFromFloat(Value=&RandomRangeFloat(Minimum=20, Maximum=40))
   )
   /SetVariables(Particles.LoopCount:NiagaraInt32=$Emitter.LoopCount) {enabled=false, isSetParameters=true, module=SetVariables_#8}
   &FN_SetSpawnGroup(Number of Spawn Groups=5)
   &SubUVAnimation(End Frame=4, SubUV Animation Mode:ENiagara_SubUVLookupMode=::NewEnumerator2)
   &InitializeRibbon(
    Lifetime=5
    Module.Write Color=true
    Module.Write Lifetime=true
    Module.Write Mass=true
    Module.Write Material Random=true
    Module.Write Position=false
    Module.Write RibbonTwist=false
    Module.Write RibbonWidth=true
    Ribbon Width=23.9844
   ) {enabled=false}
   &FN_FlipRibbonV() {enabled=false}
  ]
  particleUpdateScript [
   &ParticleState()
   /Game/Effects/Niagara/Modules/BoneFlow/FN_BoneFlow(
    Angle:NiagaraFloat=&RandomRangeFloat(Minimum=0, Maximum=1)
    BoneID:NiagaraInt32=$Particles.SpawnGroup
    Height:NiagaraFloat=&Multiply_Float(A=$Particles.NormalizedAge, B=0.85)
    Interpolate Mesh Sampling=false
    Radius:NiagaraFloat=&Multiply_Float(
     A=$Particles.NormalizedAge
     B=&FloatFromCurve(
      FloatCurve=DataInterface() {propertyValues {Curve {keys [
          {interpMode=RCIM_Cubic, tangentMode=RCTM_Break, time=0, value=1}
          {arriveTangent=-2.1505, interpMode=RCIM_Cubic, leaveTangent=-2.1505, tangentMode=RCTM_User, time=1, value=0}
         ]}}}
      Scale Curve=11
     )
    )
    SkeletalMesh:NiagaraDataInterfaceSkeletalMesh=$User.SkeletalMesh
    Write Position Attribute=true
   )
   &ScaleColor(
    Scale Mode=::NewEnumerator1
    Scale RGBA=&Vector4FromCurve(Vector4Curve=DataInterface() {propertyValues {
       AlphaCurve {keys [<at 0 0>,<at 0.1901 1>,<at 0.8127 0.5>,<at 1 0>]}
       BlueCurve {keys [<at 0.523 3>]}
       GreenCurve {keys [<at 0.523 0.547>]}
       RedCurve {keys [<at 0.523 0>]}
      }})
    ScaleRGBA=true
   )
   &SpriteRotationRate(Rotation Rate=&FloatFromCurve(
     FloatCurve=DataInterface() {propertyValues {Curve {keys [{interpMode=RCIM_Cubic, tangentMode=RCTM_Break, time=0, value=1}, {interpMode=RCIM_Cubic, tangentMode=RCTM_Break, time=1, value=0.489}]}}}
     Scale Curve=&Lerp_Float(A=-80, B=80, Alpha=$Particles.Initial.MaterialRandom)
    ))
   &CameraOffset(Camera Offset Amount=25)
   &ScaleSpriteSize(Scale Factor=&Vector2DFromFloat(Value=&FloatFromCurve(FloatCurve=DataInterface() {propertyValues {Curve {keys [{interpMode=RCIM_Cubic, leaveTangent=0.5443, tangentMode=RCTM_Break, time=0, value=0.6531}, <at 1 1>]}}})))
   &DynamicMaterialParameters(
    Index 0 Param 1=&FloatFromCurve(FloatCurve=DataInterface() {propertyValues {Curve {keys [{interpMode=RCIM_Cubic, leaveTangent=0.6389, tangentMode=RCTM_Break, time=0, value=0.5}, <at 1 1.75>]}}})
    Index 0 Param 2=&FloatFromCurve(FloatCurve=DataInterface() {propertyValues {Curve {keys [{interpMode=RCIM_Cubic, leaveTangent=0.6389, tangentMode=RCTM_Break, time=0, value=0.5}, <at 1 1.75>]}}})
    Module.Param0WriteEnabled=true
    Module.Param1WriteEnabled=true
    Module.Param2WriteEnabled=true
    Module.Param3WriteEnabled=true
   )
   /SetVariables(Particles.RibbonFacing=$Output.FN_BoneFlow.RotatedBiNormal) {enabled=false, isSetParameters=true, module=SetVariables_#9}
   &FN_RibbonMaterial(
    Color=<rgba 0 1 1 9>
    Opacity Pan U Offset:NiagaraFloat=/Game/Effects/Niagara/DynamicInputs/Random/FN_SeededRandomRangeFloat(
     Minimum:NiagaraFloat=0
     Maximum:NiagaraFloat=2
     Random Seed:NiagaraInt32=&Add_Integer(A=$Particles.LoopCount, B=$Particles.SpawnGroup)
    )
    Opacity Pan U Speed:NiagaraFloat=&FN_SeededRandomRangeFloat(
     Minimum:NiagaraFloat=0.1364
     Maximum:NiagaraFloat=0.5182
     Random Seed:NiagaraInt32=&Add_Integer(A=$Particles.LoopCount, B=$Particles.SpawnGroup)
    )
    Opacity Scale U:NiagaraFloat=&FN_SeededRandomRangeFloat(
     Minimum:NiagaraFloat=0.4273
     Maximum:NiagaraFloat=1.0032
     Random Seed:NiagaraInt32=&Add_Integer(A=$Particles.LoopCount, B=$Particles.SpawnGroup)
    )
    Scale Alpha:NiagaraFloat=&FloatFromCurve(
     FloatCurve=DataInterface() {propertyValues {Curve {keys [<at -0.007 0.0112>,<at 0.2367 0.9629>,<at 0.7537 0.9995>,<at 1 0>]}}}
     Scale Curve=&FloatFromCurve(
      FloatCurve=DataInterface() {propertyValues {Curve {keys [<at 0.0007 0.022>,<at 0.2494 1.0225>,<at 0.8089 0.9748>,<at 1 0>]}}}
      CurveIndex=$Particles.Normalized Spawn Group
     )
    )
    UV Basis:NiagaraFloat=$Particles.Normalized Spawn Group
   ) {enabled=false}
  ]
  rendererClasses [&NiagaraSpriteRendererProperties, &NiagaraRibbonRendererProperties]
  renderers [
   {rendererClass=&NiagaraSpriteRendererProperties}
   {rendererClass=&NiagaraRibbonRendererProperties, rendererIndex=1}
  ]
 }
 FN_SparkleDust_Burst {
  emitterSpawnScript [&FN_PreserveSkydiveUserVariables() {module=PreserveUserVariables}]
  emitterUpdateScript [
   &EmitterState(Life Cycle Mode=::NewEnumerator1, Loop Behavior=::NewEnumerator1, Loop Delay=$User.InitDelay, Loop Duration=1, UseLoopDelay=true)
   /Game/Effects/Niagara/Modules/Skydiving/FN_DivingAttributes/FN_EmitterDivingAttributes_v1(SkeletalMesh:NiagaraDataInterfaceSkeletalMesh=$User.SkeletalMesh)
   &SpawnBurst_Instantaneous(Spawn Count=50, Use Spawn Probability=false)
  ]
  particleSpawnScript [
   &FN_SetSpawnGroup_v2(Number of Spawn Groups=5)
   &InitializeParticle(
    Lifetime=&UniformRangedFloat(Minimum=0.2, Maximum=0.4)
    Mesh Scale=&Multiply_Vector(
     A=<1 1 1>
     B=&VectorFromFloat(Value=&UniformRangedFloat(Minimum=0.2, Maximum=1))
    )
    Module.Write Color=false
    Module.Write Lifetime=true
    Module.Write Mass=true
    Module.Write Material Random=true
    Module.Write Position=false
    Module.Write Scale=false
    Module.Write SpriteRotation=true
    Module.Write SpriteSize=true
    Position=&SimulationPosition()
    Sprite Rotation=&UniformAOrBFloat(A=0, B=360)
    Sprite Size=&Vector2DFromFloat(Value=&UniformRangedFloat(Minimum=5, Maximum=16))
   )
   &SkeletalMeshLocation(
    Bone / Socket Index=&Add_Integer(A=&Add_Integer(A=$Particles.SpawnGroup, B=4), B=15)
    Bone Sampling Mode=::NewEnumerator4
    Inherit Velocity=false
    MaskSpawnGroup=false
    Skeletal Mesh=$User.SkeletalMesh
   )
   &AddVelocity(
    Coordinate Space=::World
    Scale Added Velocity=&Lerp_Vector(A=<1 1 1>, B=<1 1 3>, Alpha=$Emitter.Dive)
    Velocity=&UniformRangedVector(Minimum=<0 0 25>, Maximum=<0 0 60>)
   ) {enabled=false}
   /Niagara/Modules/Spawn/Location/SphereLocation(MaskSpawnGroup=false, Module.Override Randomness=false, Sphere Radius=5, Surface Only=false)
   &AddVelocityFromPoint(
    Module.Use Falloff Distance=false
    Velocity Origin:NiagaraPosition=$Particles.Position
    Velocity Strength=&RandomRangeFloat(Minimum=0, Maximum=200)
   )
  ]
  particleUpdateScript [
   &ParticleState()
   /Niagara/Modules/Update/Forces/CurlNoiseForce(
    Noise Frequency:NiagaraFloat=&UniformRangedFloat(Minimum=1, Maximum=4)
    Noise Strength=&FloatFromCurve(
     FloatCurve=DataInterface() {propertyValues {Curve {keys [{interpMode=RCIM_Cubic, leaveTangent=2.5349, tangentMode=RCTM_Break, time=0, value=0}, <at 0.1288 0.8007>, <at 1 1>]}}}
     Scale Curve=&UniformRangedFloat(Minimum=50, Maximum=250)
    )
    Pan Noise=true
    Pan Noise Field=<0 0 1>
   )
   &Drag(Drag=0.5, Use Linear Drag=true, Use Rotational Drag=false)
   &GravityForce(Gravity=<0 0 50>)
   &SolveForcesAndVelocity(Clamp Velocity=false, Limit Acceleration=false)
   &ScaleSpriteSize(Scale Factor=&Multiply_Vector2(
     A=<1 1>
     B=&Vector2DFromFloat(Value=&FloatFromCurve(FloatCurve=DataInterface() {propertyValues {Curve {keys [<at -0.0046 3>,<at 1.0011 0.002>]}}}))
    ))
   /SetVariables(Particles.Color=&MakeLinearColorFromVectorAndFloat(
     "Vector (RGB)"=&MakeVectorFromLinearColorRGB(Color=&ColorFromCurve(ColorCurve=DataInterface() {propertyValues {
         AlphaCurve {keys [<at -0.0023 1 RCIM_Cubic>,<at 0.9977 1 RCIM_Cubic>]}
         BlueCurve {keys [<at 0 0.1 RCIM_Cubic>,<at 1 0>]}
         GreenCurve {keys [<at 0 0.67 RCIM_Cubic>,<at 1 0.25>]}
         RedCurve {keys [<at 0 1 RCIM_Cubic>,<at 1 0.75>]}
        }}))
     "Float (Alpha)"=&FloatFromCurve(FloatCurve=DataInterface() {propertyValues {Curve {keys [<at 0 0>,<at 0.1936 0.8315>,<at 0.4019 0.1015>,<at 0.5589 0.6516>,<at 0.7294 0.0749>,<at 0.8673 0.41>,<at 1.0011 0.002>]}}})
    )) {isSetParameters=true, module=SetVariables_#10}
  ]
  rendererClasses [&NiagaraSpriteRendererProperties]
  renderers [{rendererClass=&NiagaraSpriteRendererProperties}]
 }
 FN_SparkleDust_Constant {
  emitterSpawnScript [&FN_PreserveSkydiveUserVariables() {module=PreserveUserVariables}]
  emitterUpdateScript [
   &EmitterState(Delay First Loop Only=true, Life Cycle Mode=::NewEnumerator1, Loop Delay=$User.InitDelay, Loop Duration=1, UseLoopDelay=true)
   /Game/Effects/Niagara/Modules/Spawn/SpawnRateFromMultipleSpawnGroups_4(Global Spawnrate=5)
   &FN_EmitterDivingAttributes_v1(SkeletalMesh:NiagaraDataInterfaceSkeletalMesh=$User.SkeletalMesh)
  ]
  particleSpawnScript [
   &FN_SetSpawnGroup()
   &InitializeParticle(
    Lifetime=&UniformRangedFloat(Minimum=0.6, Maximum=1)
    Mesh Scale=&Multiply_Vector(
     A=<1 1 1>
     B=&VectorFromFloat(Value=&UniformRangedFloat(Minimum=0.2, Maximum=1))
    )
    Module.Write Color=false
    Module.Write Lifetime=true
    Module.Write Mass=true
    Module.Write Material Random=true
    Module.Write Position=false
    Module.Write Scale=false
    Module.Write SpriteRotation=true
    Module.Write SpriteSize=true
    Position=&SimulationPosition()
    Sprite Rotation=&UniformAOrBFloat(A=0, B=360)
    Sprite Size=&Vector2DFromFloat(Value=&UniformRangedFloat(Minimum=5, Maximum=14))
   )
   &SkeletalMeshLocation(
    Bone / Socket Index=&Add_Integer(A=$Particles.SpawnGroup, B=15)
    Bone Sampling Mode=::NewEnumerator4
    Inherit Velocity=false
    Interpolate Mesh Sampling=false
    MaskSpawnGroup=false
    Skeletal Mesh=$User.SkeletalMesh
   )
   &AddVelocity(
    Coordinate Space=::World
    Scale Added Velocity=&Lerp_Vector(A=<1 1 1>, B=<1 1 3>, Alpha=$Emitter.Dive)
    Velocity=&UniformRangedVector(Minimum=<0 0 25>, Maximum=<0 0 60>)
   ) {enabled=false}
   &SphereLocation(MaskSpawnGroup=false, Module.Override Randomness=false, Sphere Radius=0.56, Surface Only=false)
  ]
  particleUpdateScript [
   &ParticleState()
   &CurlNoiseForce(
    Noise Frequency:NiagaraFloat=&UniformRangedFloat(Minimum=3, Maximum=1)
    Noise Strength=&FloatFromCurve(
     FloatCurve=DataInterface() {propertyValues {Curve {keys [{interpMode=RCIM_Cubic, leaveTangent=3.1109, tangentMode=RCTM_Break, time=0, value=0}, <at 1 1>]}}}
     Scale Curve=&UniformRangedFloat(Minimum=150, Maximum=800)
    )
    Pan Noise=true
    Pan Noise Field=<0 0 1>
   )
   &Drag(Drag=0.5, Use Linear Drag=true, Use Rotational Drag=false)
   &GravityForce(Gravity=<0 0 50>)
   &SolveForcesAndVelocity(Clamp Velocity=false, Limit Acceleration=false)
   &ScaleSpriteSize(Scale Factor=&Multiply_Vector2(
     A=<1 1>
     B=&Vector2DFromFloat(Value=&FloatFromCurve(FloatCurve=DataInterface() {propertyValues {Curve {keys [
           {interpMode=RCIM_Cubic, tangentMode=RCTM_Break, time=-0.0046, value=0}
           {arriveTangent=0.0019, interpMode=RCIM_Cubic, leaveTangent=0.0019, tangentMode=RCTM_Break, time=0.1, value=3}
           {interpMode=RCIM_Cubic, tangentMode=RCTM_Break, time=1.0011, value=0.002}
          ]}}}))
    ))
   /SetVariables(Particles.Color=&MakeLinearColorFromVectorAndFloat(
     "Vector (RGB)"=&MakeVectorFromLinearColorRGB(Color=&ColorFromCurve(ColorCurve=DataInterface() {propertyValues {
         AlphaCurve {keys [<at -0.0023 0 RCIM_Cubic>,<at 0.9977 0 RCIM_Cubic>]}
         BlueCurve {keys [<at 0 0.1 RCIM_Cubic>,<at 1 0>]}
         GreenCurve {keys [<at 0 0.67 RCIM_Cubic>,<at 1 0.25>]}
         RedCurve {keys [<at 0 1 RCIM_Cubic>,<at 1 0.75>]}
        }}))
     "Float (Alpha)"=&FloatFromCurve(FloatCurve=DataInterface() {propertyValues {Curve {keys [<at 0 0>,<at 0.1936 0.8315>,<at 0.4019 0.1015>,<at 0.5589 0.6516>,<at 0.7294 0.0749>,<at 0.8673 0.41>,<at 1.0011 0.002>]}}})
    )) {isSetParameters=true, module=SetVariables_#10}
  ]
  rendererClasses [&NiagaraSpriteRendererProperties]
  renderers [{rendererClass=&NiagaraSpriteRendererProperties}]
 }
 Wispy_Burst {
  emitterUpdateScript [
   &EmitterState(
    Inactive Response=::NewEnumerator2
    Life Cycle Mode=::NewEnumerator1
    Loop Behavior=::NewEnumerator1
    Loop Delay=$User.InitDelay
    Loop Duration=0.5
    UseLoopDelay=true
   )
   &SpawnBurst_Instantaneous(Spawn Count=5, Use Spawn Probability=false)
   &SpawnRate(SpawnRate=10, Use Spawn Probability=false) {enabled=false}
  ]
  particleSpawnScript [
   &InitializeParticle(
    Color=<rgba 0 0.509 4 1>
    Lifetime=&RandomRangeFloat(Minimum=0.7, Maximum=0.8)
    Module.Write Color=true
    Module.Write Lifetime=true
    Module.Write Mass=true
    Module.Write Material Random=true
    Module.Write Position=false
    Module.Write Scale=false
    Module.Write SpriteRotation=true
    Module.Write SpriteSize=true
    Sprite Rotation=&RandomRangeFloat(Minimum=-360, Maximum=360)
    Sprite Size=&Vector2DFromFloat(Value=&RandomRangeFloat(Minimum=135, Maximum=180))
   )
   &SubUVAnimation(End Frame=4, SubUV Animation Mode:ENiagara_SubUVLookupMode=::NewEnumerator2)
   &FN_SetSpawnGroup(Number of Spawn Groups=5)
  ]
  particleUpdateScript [
   &SkeletalMeshLocation(
    Bone / Socket Index=&Add_Integer(A=&Add_Integer(A=$Particles.SpawnGroup, B=4), B=15)
    Bone Sampling Mode=::NewEnumerator4
    Inherit Velocity=false
    MaskSpawnGroup=false
    Skeletal Mesh=$User.SkeletalMesh
   )
   &ParticleState()
   &ScaleColor(
    Scale Alpha=&FloatFromCurve(FloatCurve=DataInterface() {propertyValues {Curve {keys [
         {arriveTangent=4.6219, interpMode=RCIM_Cubic, leaveTangent=4.6219, tangentMode=RCTM_User, time=0, value=0}
         {arriveTangent=-1.1136, interpMode=RCIM_Cubic, leaveTangent=-3.4217, tangentMode=RCTM_Break, time=0.1146, value=1}
         {arriveTangent=-0.5488, interpMode=RCIM_Cubic, leaveTangent=-0.5488, tangentMode=RCTM_User, time=0.1811, value=0.1682}
         <at 1 0>
        ]}}})
    ScaleA=true
    ScaleRGB=true
   )
   &CameraOffset()
   &SpriteRotationRate(Rotation Rate=&FloatFromCurve(
     FloatCurve=DataInterface() {propertyValues {Curve {keys [{interpMode=RCIM_Cubic, tangentMode=RCTM_Break, time=0, value=1}, {interpMode=RCIM_Cubic, tangentMode=RCTM_Break, time=1, value=0.65}]}}}
     Scale Curve=&Lerp_Float(A=-80, B=80, Alpha=$Particles.Initial.MaterialRandom)
    ))
   &DynamicMaterialParameters(
    Index 0 Param 1=&FloatFromCurve(FloatCurve=DataInterface() {propertyValues {Curve {keys [<at 0 0.5 RCIM_Cubic>,<at 1 1.75>]}}})
    Index 0 Param 2=&FloatFromCurve(FloatCurve=DataInterface() {propertyValues {Curve {keys [<at 0 0.5 RCIM_Cubic>,<at 1 1.75>]}}})
    Module.Param0WriteEnabled=true
    Module.Param1WriteEnabled=true
    Module.Param2WriteEnabled=true
    Module.Param3WriteEnabled=true
   )
   &ScaleSpriteSize(Scale Factor=&Vector2DFromFloat(Value=&FloatFromCurve(FloatCurve=DataInterface() {propertyValues {Curve {keys [
          {interpMode=RCIM_Cubic, leaveTangent=3.15, tangentMode=RCTM_Break, time=0, value=0}
          {arriveTangent=0.7193, interpMode=RCIM_Cubic, leaveTangent=0.7193, tangentMode=RCTM_User, time=0.3356, value=0.654}
          <at 1 1>
         ]}}})))
  ]
  rendererClasses [&NiagaraSpriteRendererProperties]
  renderers [{rendererClass=&NiagaraSpriteRendererProperties}]
 }
 Wispy_Constant {
  emitterUpdateScript [
   &EmitterState(
    Delay First Loop Only=true
    Inactive Response=::NewEnumerator2
    Life Cycle Mode=::NewEnumerator1
    Loop Delay=$User.InitDelay
    Loop Duration=0.5
    UseLoopDelay=true
   )
   &SpawnBurst_Instantaneous(Spawn Count=4, Use Spawn Probability=false) {enabled=false}
   &SpawnRate(SpawnRate=10, Use Spawn Probability=false)
  ]
  particleSpawnScript [
   &InitializeParticle(
    Color=<rgba 0 0.509 4 0.7>
    Lifetime=&RandomRangeFloat(Minimum=0.75, Maximum=1)
    Module.Write Color=true
    Module.Write Lifetime=true
    Module.Write Mass=true
    Module.Write Material Random=true
    Module.Write Position=false
    Module.Write Scale=false
    Module.Write SpriteRotation=true
    Module.Write SpriteSize=true
    Sprite Rotation=&RandomRangeFloat(Minimum=-360, Maximum=360)
    Sprite Size=&Vector2DFromFloat(Value=&RandomRangeFloat(Minimum=50, Maximum=78))
   )
   &SubUVAnimation(End Frame=4, SubUV Animation Mode:ENiagara_SubUVLookupMode=::NewEnumerator2)
   &FN_SetSpawnGroup(Number of Spawn Groups=5)
  ]
  particleUpdateScript [
   &SkeletalMeshLocation(
    Bone / Socket Index=&Add_Integer(A=&Add_Integer(A=$Particles.SpawnGroup, B=4), B=15)
    Bone Sampling Mode=::NewEnumerator4
    Inherit Velocity=false
    MaskSpawnGroup=false
    Skeletal Mesh=$User.SkeletalMesh
   )
   &ParticleState()
   &ScaleColor(
    Scale Alpha=&FloatFromCurve(FloatCurve=DataInterface() {propertyValues {Curve {keys [
         {arriveTangent=4.6219, interpMode=RCIM_Cubic, leaveTangent=4.6219, tangentMode=RCTM_User, time=0, value=0}
         {arriveTangent=-1.1136, interpMode=RCIM_Cubic, leaveTangent=-3.4217, tangentMode=RCTM_Break, time=0.1713, value=1}
         <at 1 0>
        ]}}})
    ScaleA=true
    ScaleRGB=true
   )
   &CameraOffset()
   &SpriteRotationRate(Rotation Rate=&FloatFromCurve(
     FloatCurve=DataInterface() {propertyValues {Curve {keys [{interpMode=RCIM_Cubic, tangentMode=RCTM_Break, time=0, value=1}, {interpMode=RCIM_Cubic, tangentMode=RCTM_Break, time=1, value=0.489}]}}}
     Scale Curve=&Lerp_Float(A=-80, B=80, Alpha=$Particles.Initial.MaterialRandom)
    ))
   &DynamicMaterialParameters(
    Index 0 Param 1=&FloatFromCurve(FloatCurve=DataInterface() {propertyValues {Curve {keys [{interpMode=RCIM_Cubic, leaveTangent=0.6389, tangentMode=RCTM_Break, time=0, value=0.5}, <at 1 1.75>]}}})
    Index 0 Param 2=&FloatFromCurve(FloatCurve=DataInterface() {propertyValues {Curve {keys [{interpMode=RCIM_Cubic, leaveTangent=0.6389, tangentMode=RCTM_Break, time=0, value=0.5}, <at 1 1.75>]}}})
    Module.Param0WriteEnabled=true
    Module.Param1WriteEnabled=true
    Module.Param2WriteEnabled=true
    Module.Param3WriteEnabled=true
   )
   &ScaleSpriteSize(Scale Factor=&Vector2DFromFloat(Value=&FloatFromCurve(FloatCurve=DataInterface() {propertyValues {Curve {keys [{interpMode=RCIM_Cubic, leaveTangent=0.0179, tangentMode=RCTM_Break, time=0, value=0.1498}, <at 1 1>]}}})))
  ]
  rendererClasses [&NiagaraSpriteRendererProperties]
  renderers [{rendererClass=&NiagaraSpriteRendererProperties}]
 }
}
summary {
 systemName=NS_Glider_Cherry_Idle
 emitters [
  {emitterName=Corona_Constant, rendererClasses [&NiagaraSpriteRendererProperties]}
  {emitterName=FN_SparkleDust_Constant, rendererClasses [&NiagaraSpriteRendererProperties]}
  {emitterName=FN_AttachedRibbon_v3_HANDS_2, rendererClasses [&NiagaraRibbonRendererProperties]}
  {emitterName=Wispy_Constant, rendererClasses [&NiagaraSpriteRendererProperties]}
  {
   emitterName=FN_BoneFlow_v5_BETA
   rendererClasses [&NiagaraSpriteRendererProperties, &NiagaraRibbonRendererProperties]
  }
  {emitterName=Corona_BURST, rendererClasses [&NiagaraSpriteRendererProperties]}
  {emitterName=Wispy_Burst, rendererClasses [&NiagaraSpriteRendererProperties]}
  {emitterName=FN_SparkleDust_Burst, rendererClasses [&NiagaraSpriteRendererProperties]}
  {emitterName=FN_BoneFlow_Sprite_HAIR, rendererClasses [&NiagaraSpriteRendererProperties]}
  {emitterName=FN_BoneFlow_HEAD, rendererClasses [&NiagaraRibbonRendererProperties]}
 ]
 userVariables [
  User.bIsFrontEndPreview {type=NiagaraBool}
  User.ForwardDot {type=NiagaraFloat, defaultValue=0, description=""}
  User.InitDelay {type=NiagaraFloat, defaultValue=0.825, description=""}
  User.PlayerSpeed {type=NiagaraFloat, defaultValue=0, description=""}
  User.RibbonIDOffset {type=NiagaraInt32}
  User.SkeletalMesh {
   type=NiagaraDataInterfaceSkeletalMesh
   defaultValue {
    struct=NiagaraExt_VariableValue_DataInterface
    value {dataInterface=&self:NiagaraDataInterfaceSkeletalMesh_3}
   }
  }
  User.SkeletonHair {
   type=NiagaraDataInterfaceSkeletalMesh
   defaultValue {
    struct=NiagaraExt_VariableValue_DataInterface
    value {dataInterface=&self:NiagaraDataInterfaceSkeletalMesh_1}
   }
  }
 ]
}
